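# Niagara death effect for a large spire monolith: a burst of smoke, debris, energy and shockwaves.
# Emitters layer dense center smoke, radial smoke, GPU dissolve particles, mesh debris, small lightning meshes, a lit energy wave, a mesh ring and a post-process-style shockwave sprite.
# Age curves drive dynamic material parameters for color, erosion and distortion.
# User parameters scale size, velocity and shockwave strength, while camera distance adjusts smoke offsets.
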NiagaraSystem /PistolCanary/Assets/FX/Monoliths/MonolithDeath/NS_LargeSpire_DeathFX
dependencies {
 usedDataInterfaces [
  &NiagaraDataInterfaceCurve
  &NiagaraDataInterfaceCamera
  &NiagaraDataInterfaceColorCurve
  &NiagaraDataInterfaceCollisionQuery
  &NiagaraDataInterfaceLandscape
 ]
 usedDynamicInputs []
 usedModules [&SystemState]
}
emitters {
 CenterDenseSmoke {
  emitterUpdateScript [
   &EmitterState(Life Cycle Mode=::NewEnumerator1, Loop Behavior=::NewEnumerator1, Loop Duration=1, UseLoopDelay=true)
   &SpawnRate(
    SpawnRate=&FloatFromCurve(FloatCurve=DataInterface() {propertyValues {Curve {keys [<at 0 1>,<at 1 0.3>]}}}, CurveIndex=$Emitter.NormalizedLoopAge, Scale Curve=100)
    Use Spawn Probability=false
   ) {enabled=false}
   &SpawnBurst_Instantaneous(Spawn Count=10, Use Spawn Probability=false, UseLoopCountLimit=false)
   /SetVariables(Emitter.LightingCenter:Vector4f=&MakeFloat4FromVectorAndFloat("Vector (RGB)"=/Niagara/DynamicInputs/Helpers/SimulationPosition(), "Float (Alpha)"=1)) {isSetParameters=true}
  ]
  particleSpawnScript [
   &InitializeParticle(
    Color=&MultiplyLinearColor(
     Linear Color=&MaskLinearColorBySpawnGroup(
      Masked Linear Color=<rgba 1 1 1 1>
      Passthrough Linear Color=&MaskLinearColorBySpawnGroup(Masked Linear Color=<rgba 1 1 1 1>, Passthrough Linear Color=<rgba 1 1 1 1>)
      Spawn Group Mask=1
     )
     Scale Factor=&MakeVector4_FromLinearColor(LinearColor=$User.SmokeColorMul)
    )
    Lifetime=&RandomRangeFloat(Minimum=0.5, Maximum=1)
    Sprite Rotation Mode=::NewEnumerator2
    Sprite Size Mode=::NewEnumerator3
    Uniform Sprite Size=&Multiply_Float(
     A=&Multiply_Float(
      A=&RandomRangeFloat(Minimum=300, Maximum=700)
      B=&MaskFloatBySpawnGroup(
       Masked Float=1
       Passthrough Float=&MaskFloatBySpawnGroup(Masked Float=1.3, Passthrough Float=1, Spawn Group Mask=2)
       Spawn Group Mask=1
      )
     )
     B=&Multiply_Float(A=$User.ElementSizeScale, B=$User.SmokeSizeScale)
    )
    UsePositionOffset=true
    Write Material Random=true
   ) {module=InitializeParticle001}
   &ShapeLocation(
    MaskSpawnGroup=false
    Non Uniform Scale=<2 2 7>
    Override Seed=false
    Sphere Radius=&Multiply_Float(A=100, B=&Multiply_Float(A=$User.ElementSizeScale, B=$User.SmokeSizeScale))
    UseOwnerScale=true
   )
   &AddVelocity(
    Module.Use Falloff Distance=false
    Override Seed=false
    Velocity Mode=::NewEnumerator1
    Velocity Speed=&RandomRangeFloat(Minimum=200, Maximum=2346.4463)
   )
   /SetVariables(Particles.Velocity=&Multiply_Vector(A=$Particles.Velocity, B=&VectorFromFloat(Value=$User.VelocityScale))) {isSetParameters=true}
  ]
  particleUpdateScript [
   &ParticleState()
   &CameraOffset(Camera Offset Amount=&Multiply_Float(
     A=&RandomRangeFloat(Minimum=25, Maximum=75)
     B=&FloatFromCurve(
      FloatCurve=DataInterface() {propertyValues {Curve {keys [<at 0 0.3>,<at 1 1>]}}}
      CurveIndex=/Game/Effects/Niagara/Modules/Math/FN_CameraDistance(
       0Dist:NiagaraFloat=200
       1Dist:NiagaraFloat=700
       Camera Query:NiagaraDataInterfaceCamera=DataInterface()
       Position:NiagaraPosition=$Particles.Position
      )
     )
    ))
   &ScaleSpriteSize(Scale Factor=<1.2 1>, Scale Sprite Size Mode=::NewEnumerator2) {enabled=false, module=ScaleSpriteSize001}
   &Drag(Drag=3.5, Use Linear Drag=true, Use Rotational Drag=false)
   &GravityForce(Gravity=<0 0 -386.461>) {enabled=false}
   &CurlNoiseForce(
    Bias Noise=false
    Noise Frequency=4
    Noise Strength=&Multiply_Float(A=2931.5151, B=$User.VelocityScale)
    Pan Noise=true
    Pan Noise Field=<0 0 2>
   )
   &PointForce(
    Force Falloff Distance=242950.5781
    Force Strength=&FloatFromCurve(
     FloatCurve=DataInterface() {propertyValues {Curve {keys [<at 0.5 0>,<at 0.7 1>]}}}
     Scale Curve=&Multiply_Float(A=5053.7715, B=$User.VelocityScale)
    )
    Module.Use Falloff=true
    Module.UseLinearForce=false
   )
   &SolveForcesAndVelocity(Clamp Velocity=false, Limit Acceleration=false)
   &ScaleSpriteSize(Scale Factor=&Vector2DFromFloat(Value=&FloatFromCurve(
      FloatCurve=DataInterface() {propertyValues {Curve {keys [
          {interpMode=RCIM_Cubic, leaveTangent=1.7304, tangentMode=RCTM_Break, time=0, value=0.8}
          {arriveTangent=-0.0555, interpMode=RCIM_Cubic, leaveTangent=-0.0555, tangentMode=RCTM_User, time=0.3, value=1}
          <at 1 1.2>
         ]}}}
      Scale Curve=4
     )))
   /SetVariables(Particles.Color_ParticleEnergy:LinearColor=&ColorFromCurve(ColorCurve=DataInterface() {propertyValues {
       AlphaCurve {keys [<at 0 1 RCIM_Cubic>]}
       BlueCurve {keys [<at 0.3364 2>,<at 0.749 0.4167 RCIM_Cubic>]}
       GreenCurve {keys [<at 0.3364 0.35>,<at 0.749 0 RCIM_Cubic>]}
       RedCurve {keys [<at 0.3364 0.2>,<at 0.749 5 RCIM_Cubic>]}
      }})) {isSetParameters=true}
   &DynamicMaterialParameters(
    Index 0 Param 1=&FloatFromCurve(FloatCurve=DataInterface() {propertyValues {Curve {keys [{tangentMode=RCTM_User, time=0, value=0}, {tangentMode=RCTM_User, time=1, value=1}]}}})
    Index 0 Param 3=&FloatFromCurve(
     FloatCurve=DataInterface() {propertyValues {Curve {keys [
         {interpMode=RCIM_Cubic, tangentMode=RCTM_Break, time=0, value=0}
         {interpMode=RCIM_Cubic, leaveTangent=-3.8707, tangentMode=RCTM_Break, time=0.2, value=1}
         {tangentMode=RCTM_User, time=1, value=0}
        ]}}}
     Scale Curve=0
    )
    Index 0 Param 4=&Multiply_Float(
     A=&Waveform("Waveform [1]"=::NewEnumerator12, "[1] Frequency"=13, "[1] Amplitude Min/Max"=<1 4>)
     B=&FloatFromCurve(
      FloatCurve=DataInterface() {propertyValues {Curve {keys [{interpMode=RCIM_Cubic, leaveTangent=-4.1698, tangentMode=RCTM_Break, time=1.1, value=1}, <at 1.5 0>]}}}
      CurveIndex=$Emitter.Age
     )
    )
    Index 1 Param 1=&MakeFloatFromVector(Vector=&MakeVectorFromLinearColorRGB(Color=$Particles.Color_ParticleEnergy), Channel=::NewEnumerator0)
    Index 1 Param 2=&MakeFloatFromVector(Vector=&MakeVectorFromLinearColorRGB(Color=$Particles.Color_ParticleEnergy), Channel=::NewEnumerator1)
    Index 1 Param 3=&MakeFloatFromVector(Vector=&MakeVectorFromLinearColorRGB(Color=$Particles.Color_ParticleEnergy), Channel=::NewEnumerator2)
    Index 1 Param 4=&FloatFromCurve(FloatCurve=DataInterface() {propertyValues {Curve {keys [
         {interpMode=RCIM_Cubic, leaveTangent=0.1782, tangentMode=RCTM_Break, time=0, value=0.2}
         {arriveTangent=2.1223, tangentMode=RCTM_User, time=1, value=1.2}
        ]}}})
    Index 2 Param 1=100
    Index 2 Param 2=&Multiply_Float(
     A=10000
     B=&FloatFromCurve(
      FloatCurve=DataInterface() {propertyValues {Curve {keys [{interpMode=RCIM_Cubic, tangentMode=RCTM_Break, time=0, value=2}, <at 0.7 1>]}}}
      CurveIndex=$Emitter.Age
     )
    )
    Index 2 Param 3=&Multiply_Float(
     A=684.907
     B=&FloatFromCurve(
      FloatCurve=DataInterface() {propertyValues {Curve {keys [{interpMode=RCIM_Cubic, tangentMode=RCTM_Break, time=0, value=2}, <at 0.7 1>]}}}
      CurveIndex=$Emitter.Age
     )
    )
    Index 2 Param 4=&FloatFromCurve(FloatCurve=DataInterface() {propertyValues {Curve {keys [
         {interpMode=RCIM_Cubic, tangentMode=RCTM_Break, time=0.5, value=0}
         {arriveTangent=2.6267, interpMode=RCIM_Cubic, tangentMode=RCTM_Break, time=1, value=0.8}
        ]}}})
    Index 3 Param 1=&FloatFromCurve(FloatCurve=DataInterface() {propertyValues {Curve {keys [<at 0 1>,<at 0.7 1>,<at 2 1>]}}}, CurveIndex=$Engine.System.Age)
    Module.Index0Param0Write=true
    Module.Index0Param1Write=true
    Module.Index0Param2Write=true
    Module.Index0Param3Write=true
    Module.Index1Param0Write=true
    Module.Index1Param1Write=true
    Module.Index1Param2Write=true
    Module.Index1Param3Write=true
    Module.Index2Param0Write=true
    Module.Index2Param1Write=true
    Module.Index2Param2Write=true
    Module.Index2Param3Write=true
    Module.Index3Param0Write=true
    Module.Index3Param1Write=true
    Module.Index3Param2Write=true
    Module.Index3Param3Write=true
    Write Parameter Index 1=true
    Write Parameter Index 2=true
    Write Parameter Index 3=true
   )
   &ScaleColor(
    Scale Alpha=&FloatFromCurve(
     FloatCurve=DataInterface() {propertyValues {Curve {keys [
         {interpMode=RCIM_Cubic, leaveTangent=1.3072, tangentMode=RCTM_Break, time=0, value=0}
         {interpMode=RCIM_Cubic, tangentMode=RCTM_Break, time=0.25, value=1}
         {arriveTangent=-4.4597, interpMode=RCIM_Cubic, leaveTangent=-4.4597, tangentMode=RCTM_User, time=1, value=0}
        ]}}}
     Scale Curve=&FloatFromCurve(FloatCurve=DataInterface() {propertyValues {Curve {keys [<at 1.3 1>,<at 1.8 0>]}}}, CurveIndex=$Emitter.Age)
    )
    ScaleA=true
    ScaleRGB=true
   )
  ]
  rendererClasses [&NiagaraSpriteRendererProperties]
  renderers [{rendererClass=&NiagaraSpriteRendererProperties}]
 }
 DissoveParticle001 {
  simTarget=GPUComputeSim
  emitterUpdateScript [
   &EmitterState(Life Cycle Mode=::NewEnumerator1, Loop Behavior=::NewEnumerator1, Loop Duration=1, UseLoopDelay=true)
   &SpawnBurst_Instantaneous(Spawn Count=300, Use Spawn Probability=false, UseLoopCountLimit=false)
   &SpawnRate(SpawnRate=600, Use Spawn Probability=false) {enabled=false, module=SpawnRate001}
   &SpawnRate(
    SpawnRate=&FloatFromCurve(
     FloatCurve=DataInterface() {propertyValues {Curve {keys [
         {interpMode=RCIM_Cubic, tangentMode=RCTM_Break, time=0, value=0}
         {arriveTangent=0.7041, interpMode=RCIM_Cubic, leaveTangent=0.7041, tangentMode=RCTM_User, time=0.3, value=0.05}
         {arriveTangent=2.4035, tangentMode=RCTM_User, time=1, value=1}
        ]}}}
     CurveIndex=$Emitter.NormalizedLoopAge
     Scale Curve=376.5758
    )
    Use Spawn Probability=false
   ) {enabled=false}
  ]
  particleSpawnScript [
   &InitializeParticle(
    Color=<rgba 0.4 0.194 0.2112 1>
    Lifetime Max=1.54
    Lifetime Min=0.5
    Lifetime Mode=::NewEnumerator1
    Mesh Scale Mode=::NewEnumerator3
    Mesh Uniform Scale=100
    Sprite Rotation Mode=::NewEnumerator2
    Sprite Size Mode=::NewEnumerator3
    Uniform Sprite Size=&FloatFromCurve(
     FloatCurve=DataInterface() {propertyValues {Curve {keys [<at 0 2.5>,<at 0.13 1>,<at 1 0.3>]}}}
     CurveIndex=&RandomRangeFloat(Minimum=0, Maximum=1)
     Scale Curve=195.0603
    )
    UsePositionOffset=false
    Write Material Random=true
   )
   /SetVariables() {isSetParameters=true}
   &AddVelocity(
    Module.Use Falloff Distance=false
    Override Seed=false
    Velocity Mode=::NewEnumerator1
    Velocity Speed=&Multiply_Float(
     A=&RandomRangeFloat(Minimum=2500, Maximum=6000)
     B=&FloatFromCurve(
      FloatCurve=DataInterface() {propertyValues {Curve {keys [
          {interpMode=RCIM_Cubic, leaveTangent=-1.3134, tangentMode=RCTM_Break, time=0.4, value=1}
          {arriveTangent=0.0485, tangentMode=RCTM_User, time=1, value=0.8}
         ]}}}
      CurveIndex=$Emitter.Age
     )
    )
   )
   &ShapeLocation(
    Cylinder Height=150.0747
    Cylinder Radius=32.9399
    MaskSpawnGroup=false
    Override Seed=false
    Shape Primitive=::NewEnumerator1
    Surface Only=false
    UseOwnerScale=true
   )
   &InitialMeshOrientation(Use Rotation Vector=false) {enabled=false}
   &CameraOffset(Camera Offset Amount=-100)
  ]
  particleUpdateScript [
   &ParticleState()
   &ScaleSpriteSize(Scale Sprite Size Mode=::NewEnumerator2) {module=ScaleSpriteSize001}
   &ScaleSpriteSize(Scale Factor=&Vector2DFromFloat(Value=&FloatFromCurve(
      FloatCurve=DataInterface() {propertyValues {Curve {keys [
          {arriveTangent=4.0375, interpMode=RCIM_Cubic, leaveTangent=4.0375, tangentMode=RCTM_User, time=0, value=0}
          {arriveTangent=-0.0952, interpMode=RCIM_Cubic, leaveTangent=-0.0952, tangentMode=RCTM_User, time=0.2, value=0.3973}
         ]}}}
      Scale Curve=$User.ElementSizeScale
     )))
   /Niagara/Modules/Update/SubUV/SubUVAnimation(Number Of Frames=36) {enabled=false}
   &GravityForce(Gravity=<0 0 250>)
   &WindForce(
    Wind Speed=&TransformVector(
     Vector=&RandomRangeVector(Minimum=<239.2935 0 0>, Maximum=<1422.8411 1 1>)
     Source Space=::World
     Destination Space=::Simulation
    )
    Wind Speed Scale=&FloatFromCurve(
     FloatCurve=DataInterface() {propertyValues {Curve {keys [{interpMode=RCIM_Cubic, tangentMode=RCTM_Break, time=0, value=0}, {arriveTangent=2.6532, tangentMode=RCTM_User, time=1, value=1}]}}}
     CurveIndex=$Emitter.Age
    )
   ) {enabled=false}
   &CurlNoiseForce(
    Bias Noise=false
    Noise Frequency=3
    Noise Strength=&FloatFromCurve(
     FloatCurve=DataInterface() {propertyValues {Curve {keys [{interpMode=RCIM_Cubic, tangentMode=RCTM_Break, time=0, value=1}, {arriveTangent=3.0586, tangentMode=RCTM_User, time=1, value=2}]}}}
     CurveIndex=$Emitter.Age
     Scale Curve=&RandomRangeFloat(Minimum=500, Maximum=1800)
    )
    Pan Noise=true
    Pan Noise Field=<0 0 2>
   )
   &Drag(Drag=&RandomRangeFloat(Minimum=2.4744, Maximum=3.9427), Use Linear Drag=true, Use Rotational Drag=true)
   &SolveForcesAndVelocity(Clamp Velocity=false, Limit Acceleration=false)
   &ScaleColor(
    Scale Alpha=&FloatFromCurve(FloatCurve=DataInterface() {propertyValues {Curve {keys [{interpMode=RCIM_Cubic, tangentMode=RCTM_Break, time=0.2, value=1}, <at 1 0>]}}})
    ScaleA=true
    ScaleRGB=true
   )
   &DynamicMaterialParameters(
    Index 0 Param 1=&RandomRangeFloat(Minimum=0, Maximum=100)
    Index 0 Param 2=&FloatFromCurve(FloatCurve=DataInterface() {propertyValues {Curve {keys [{interpMode=RCIM_Cubic, tangentMode=RCTM_Break, time=0, value=-0.9}, <at 1 1>]}}})
    Index 0 Param 3=&FloatFromCurve(FloatCurve=DataInterface() {propertyValues {Curve {keys [<at 0.2 1>,<at 0.4 0.5>]}}})
    Module.Index0Param0Write=true
    Module.Index0Param1Write=true
    Module.Index0Param2Write=true
    Module.Index0Param3Write=true
   )
   &FN_Sprite_Attributes(
    Color=<rgba 1 0.01 0.0265 0.05>
    Create Color Attribute=true
    Create SpriteSize Attribute=true
    SpriteSize=<366.7501 353.7109>
    Write to Module Namespace=true
   )
  ]
  rendererClasses [&NiagaraMeshRendererProperties, &NiagaraSpriteRendererProperties]
  renderers [
   {rendererClass=&NiagaraMeshRendererProperties, rendererIndex=0}
   {rendererClass=&NiagaraSpriteRendererProperties, rendererIndex=1}
   {rendererClass=&NiagaraSpriteRendererProperties, rendererIndex=2}
  ]
 }
 EnergyWave001 {
  emitterUpdateScript [
   &EmitterState()
   &SpawnBurst_Instantaneous(Use Spawn Probability=false, UseLoopCountLimit=false)
  ]
  particleSpawnScript [
   /SetVariables(
    Particles.FlareCamOffset=200
    Particles.FlareDynMaterialParm:Vector4f=&MakeVector4(X=0.2, Y=1, Z=0, W=0)
    Particles.FlareSpriteSize:Vector2f=&Vector2DFromFloat(Value=&Multiply_Float(A=700, B=$System.Global Scale))
   ) {isSetParameters=true}
   &InitializeParticle(
    Color Mode=::NewEnumerator0
    Lifetime=0.7
    Mesh Scale Mode=::NewEnumerator3
    Mesh Uniform Scale=11.5
    Sprite Size Mode=::NewEnumerator3
    Uniform Sprite Size=2500
    UsePositionOffset=true
    Write Material Random=true
   )
   &InitialMeshOrientation(
    Mesh Axis To Orient=<0 0 0>
    Mesh Orientation Mode=::NewEnumerator2
    Orientation Vector=<0 0 1>
    Rotation=<0 1 1>
    Rotation Coordinate Space=::NewEnumerator0
    Use Rotation Vector=true
   ) {enabled=false}
   &CameraOffset(Camera Offset Amount=100)
  ]
  particleUpdateScript [
   &ParticleState()
   &ScaleColor(
    Linear Color Curve=DataInterface() {propertyValues {
      AlphaCurve {keys [<at 0 1>,<at 0.1637 1>,<at 1 0>]}
      BlueCurve {keys [<at 0 0.0833>,<at 0.1684 0.0833>,<at 0.3654 0.0833>,<at 1 0.0833>]}
      GreenCurve {keys [<at 0 0>,<at 0.1684 0>,<at 0.3654 0>,<at 1 0>]}
      RedCurve {keys [<at 0 1>,<at 0.1684 1>,<at 0.3654 1>,<at 1 1>]}
     }}
    Scale Mode=::NewEnumerator2
   )
   &DynamicMaterialParameters(
    Index 0 Param 1=&FloatFromCurve(FloatCurve=DataInterface() {propertyValues {Curve {keys [{arriveTangent=-2.1015, interpMode=RCIM_Cubic, leaveTangent=-2.1015, tangentMode=RCTM_User, time=0, value=0.5}, <at 1 -0.15>]}}})
    Index 0 Param 2=&FloatFromCurve(FloatCurve=DataInterface() {propertyValues {Curve {keys [{arriveTangent=-0.3464, interpMode=RCIM_Cubic, leaveTangent=-0.3464, tangentMode=RCTM_User, time=0, value=1}, <at 1 0.9>]}}})
    Index 0 Param 3=&FloatFromCurve(FloatCurve=DataInterface() {propertyValues {Curve {keys [
         {interpMode=RCIM_Cubic, tangentMode=RCTM_Break, time=0, value=0}
         {arriveTangent=0.4478, interpMode=RCIM_Cubic, leaveTangent=0.4478, tangentMode=RCTM_User, time=0.15, value=1}
         {interpMode=RCIM_Cubic, tangentMode=RCTM_Break, time=1, value=1.2}
        ]}}})
    Module.Param0WriteEnabled=true
    Module.Param1WriteEnabled=true
    Module.Param2WriteEnabled=true
    Module.Param3WriteEnabled=false
   )
   &ScaleColor(
    Scale Alpha=0.5
    Scale RGB=&VectorFromFloat(Value=&FloatFromCurve(FloatCurve=DataInterface() {propertyValues {Curve {keys [
          {interpMode=RCIM_Cubic, tangentMode=RCTM_Break, time=0, value=10}
          {arriveTangent=-6.6072, interpMode=RCIM_Cubic, leaveTangent=-6.6072, tangentMode=RCTM_User, time=0.15, value=3}
          {interpMode=RCIM_Cubic, tangentMode=RCTM_Break, time=1, value=1}
         ]}}}))
    ScaleA=true
    ScaleRGB=true
   ) {module=ScaleColor001}
   &ScaleMeshSize(Scale Factor=&VectorFromFloat(Value=&FloatFromCurve(
      FloatCurve=DataInterface() {propertyValues {Curve {keys [
          {interpMode=RCIM_Cubic, tangentMode=RCTM_Break, time=0, value=0.2}
          {arriveTangent=2.9365, interpMode=RCIM_Cubic, leaveTangent=2.9365, tangentMode=RCTM_User, time=0.1, value=1}
          {interpMode=RCIM_Cubic, tangentMode=RCTM_Break, time=1, value=2}
         ]}}}
      Scale Curve=1.2
     )))
   &ScaleSpriteSize(Uniform Curve Sprite Scale=DataInterface() {propertyValues {Curve {keys [
        {interpMode=RCIM_Cubic, tangentMode=RCTM_Break, time=0, value=0.2}
        {arriveTangent=2.9365, interpMode=RCIM_Cubic, leaveTangent=2.9365, tangentMode=RCTM_User, time=0.1, value=1}
        {interpMode=RCIM_Cubic, tangentMode=RCTM_Break, time=1, value=2}
       ]}}})
   &FN_Light_Attributes(
    Color=<rgba 1 0.0638 0.2759 1>
    Create Color Attribute=true
    Create Exponent Attribute=true
    Create Position Attribute=true
    Create Radius Attribute=true
    Exponent=20
    Position=<0 0 50>
    Radius=800
    Scale Color=&FloatFromCurve(
     FloatCurve=DataInterface() {propertyValues {Curve {keys [
         <at 0 0 RCIM_Cubic>
         {interpMode=RCIM_Cubic, leaveTangentWeight=0.9776, tangentMode=RCTM_Break, time=0.3, value=1}
         {arriveTangent=-2.4571, interpMode=RCIM_Cubic, tangentMode=RCTM_Break, time=1, value=0}
        ]}}}
     Scale Curve=10
    )
   )
   /SetVariables(Particles.FlareColor=&MultiplyLinearColor(
     Linear Color=<rgba 8.75 0 15 1>
     Scale Factor=&MakeFloat4FromVectorAndFloat(
      "Vector (RGB)"=<1 1 1>
      "Float (Alpha)"=&FloatFromCurve(
       FloatCurve=DataInterface() {propertyValues {Curve {keys [{interpMode=RCIM_Cubic, tangentMode=RCTM_User, time=0, value=1}, {interpMode=RCIM_Cubic, tangentMode=RCTM_User, time=0.5, value=0}]}}}
       Scale Curve=&MakeCustomFloatFromBool(
        Bool=$User.UseNearCameraFade
        True Float=&Lerp_Float(
         A=0
         B=1
         Alpha=&FloatFromCurve(
          FloatCurve=DataInterface() {propertyValues {Curve {keys [<at 0 0 RCIM_Cubic>, {arriveTangent=3.6356, interpMode=RCIM_Cubic, leaveTangent=3.6356, tangentMode=RCTM_User, time=1, value=1}]}}}
          CurveIndex=$System.ScaleDistance
         )
        )
        False Float=1
       )
      )
     )
    )) {isSetParameters=true}
  ]
  rendererClasses [
   &NiagaraMeshRendererProperties
   &NiagaraLightRendererProperties
   &NiagaraSpriteRendererProperties
  ]
  renderers [
   {rendererClass=&NiagaraMeshRendererProperties, rendererIndex=0}
   {rendererClass=&NiagaraLightRendererProperties, rendererIndex=1}
   {rendererClass=&NiagaraSpriteRendererProperties, rendererIndex=2}
  ]
 }
 FN_Debris_Mesh {
  emitterUpdateScript [
   &EmitterScalability(
    Module.bApplySpawnCountScale=false
    Module.bApplySpawnCountScaleByDistanceFraction=false
    Module.bUseMaxDistance=false
    Module.bUseMinDistance=false
   ) {enabled=false}
   &EmitterLifeCycle() {enabled=false}
   &EmitterState()
   &SpawnBurst_Instantaneous(Spawn Count=50, Use Spawn Probability=false)
  ]
  particleSpawnScript [
   /Niagara/Modules/Spawn/Initialization/InitializeParticle(
    Lifetime=&RandomRangeFloat(Minimum=0.3, Maximum=0.7)
    Mesh Scale=&Multiply_VectorByFloat(
     Vector=&RandomRangeVector(Minimum=<0.1 0.1 0.1>, Maximum=<1 1 1>)
     Float=&FloatFromCurve(
      FloatCurve=DataInterface() {propertyValues {Curve {keys [{arriveTangent=-1.5969, interpMode=RCIM_Cubic, leaveTangent=-1.5969, tangentMode=RCTM_User, time=0, value=1}, <at 1 0.25 RCIM_Cubic>]}}}
      CurveIndex=&RandomRangeFloat(Minimum=0.5, Maximum=1)
      Scale Curve=100
     )
    )
    Module.Write Color=true
    Module.Write Lifetime=true
    Module.Write Mass=true
    Module.Write Material Random=true
    Module.Write Position=false
    Module.Write Scale=true
    Module.Write SpriteRotation=false
    Module.Write SpriteSize=false
    Position=/Niagara/Functions/Localspace/SimulationPosition()
   )
   &ShapeLocation(MaskSpawnGroup=false, Non Uniform Scale=<2 2 10>, Override Seed=false, Sphere Radius=50, Sphere Surface Distribution=1, UseOwnerScale=true)
   /Niagara/Modules/Spawn/Orientation/V2/InitialMeshOrientation(
    Orientation Vector=/Niagara/DynamicInputs/Random/RandomVector(
     Vector Scale:NiagaraFloat=1
     Constraint Vector:Vector3f=<0 0 1>
     Randomness Mode:ENiagaraRandomnessMode=::NewEnumerator0
     Random Seed:NiagaraInt32=0
     Fixed Random Seed:NiagaraBool=false
    )
    Use Orientation Vector=true
    Use Rotation Vector=false
   )
   &MeshRotationForce(
    "Lever Radius (cm)":NiagaraFloat=&UniformRangedFloat(Minimum=200, Maximum=400)
    Rotation=&RandomVector(
     Vector Scale:NiagaraFloat=1
     Constraint Vector:Vector3f=<0 0 1>
     Randomness Mode:ENiagaraRandomnessMode=::NewEnumerator0
     Random Seed:NiagaraInt32=0
     Fixed Random Seed:NiagaraBool=false
    )
   )
   &AddVelocity(
    Velocity=$StackContext.ShapeLocation.ShapeNormal
    Velocity Speed Scale=&RandomRangeFloat(Minimum=500, Maximum=2000)
   )
   &AddVelocity(Rotation Coordinate Space=::World, Velocity=<1 1 1>, Velocity Speed Scale=200) {module=AddVelocity001}
   &AddVelocityInCone(
    Cone Angle=100
    Cone Axis=<0 0 1>
    Module.Use Velocity Falloff On Cone Axis=false
    Velocity Strength=&UniformRangedFloat(Minimum=300, Maximum=30000)
   ) {enabled=false}
   &ApplyInitialForces()
   /SetVariables(Particles.RandomGravity:NiagaraFloat=&UniformRangedFloat(Minimum=-20000, Maximum=-9821)) {enabled=false, isSetParameters=true}
  ]
  particleUpdateScript [
   &ParticleState()
   &ScaleMeshSize(Scale Factor=&VectorFromFloat(Value=&FloatFromCurve(FloatCurve=DataInterface() {propertyValues {Curve {keys [<at 0.6 1>,<at 1 0>]}}})))
   &AccelerationForce(Acceleration=&MakeVector(X=1, Y=9, Z=$Particles.RandomGravity)) {enabled=false}
   &Drag(
    Drag=&FloatFromCurve(FloatCurve=DataInterface() {propertyValues {Curve {keys [
         <at 0.2 1 RCIM_Cubic>
         {arriveTangent=-0.0326, interpMode=RCIM_Cubic, leaveTangent=-0.0326, tangentMode=RCTM_User, time=0.3, value=0}
         <at 1 0 RCIM_Cubic>
        ]}}})
    Use Linear Drag=true
    Use Rotational Drag=true
   )
   &SolveForcesAndVelocity(Clamp Velocity=false, Limit Acceleration=false)
   &SolveRotationalForcesAndVelocity()
   &Color()
  ]
  rendererClasses [&NiagaraMeshRendererProperties]
  renderers [{rendererClass=&NiagaraMeshRendererProperties, rendererIndex=0}]
 }
 LightningSmall_BodyRound {
  emitterUpdateScript [
   &EmitterState(Inactive Response=::NewEnumerator1, Life Cycle Mode=::NewEnumerator1, Loop Behavior=::NewEnumerator1, Loop Duration=1, UseLoopDelay=true)
   &SpawnBurst_Instantaneous(Spawn Count=2, Use Spawn Probability=false, UseLoopCountLimit=false) {enabled=false}
   &SpawnRate(SpawnRate=10, Use Spawn Probability=true)
  ]
  particleSpawnScript [
   /Niagara/Modules/Spawn/Initialization/V2/InitializeParticle(
    Color Maximum=<rgba 1000 134.5969 36.5312 1>
    Color Minimum=<rgba 100 1.4573 27.4752 1>
    Color Mode=::NewEnumerator2
    Lifetime Max=0.7
    Lifetime Min=0.4
    Lifetime Mode=::NewEnumerator1
    Mesh Scale=&RandomRangeVector(Minimum=<5 5 5>, Maximum=<29 29 29>)
    Mesh Scale Mode=::NewEnumerator4
    UsePositionOffset=false
    Write Material Random=true
   )
   &CameraOffset(Camera Offset Amount=5)
   &ShapeLocation(
    MaskSpawnGroup=false
    Offset=&RandomRangeVector(Minimum=<0 0 60>, Maximum=<0 0 150>)
    Offset Mode=::NewEnumerator0
    Override Seed=false
    Sphere Radius=500
    UseOwnerScale=true
   )
   /Niagara/Modules/Spawn/Orientation/V3/InitialMeshOrientation(Mesh Orientation Mode=::NewEnumerator2, Rotation=<0 0 0.25>, Use Rotation Vector=true) {module=InitialMeshOrientation002}
   &InitialMeshOrientation(
    Mesh Orientation Mode=::NewEnumerator2
    Rotation=&RandomRangeVector(Minimum=<0 0 0>, Maximum=<1 0 0>)
    Use Rotation Vector=true
   )
   &AddVelocity(Module.Use Falloff Distance=false, Origin Offset=<0 0 100>, Override Seed=false, Velocity Mode=::NewEnumerator1, Velocity Speed=1)
  ]
  particleUpdateScript [
   &ParticleState()
   &ScaleMeshSize(Scale Factor=&Multiply_VectorByFloat(Vector=<1 1 1>, Float=1))
   &ScaleColor(
    Scale Alpha=&Multiply_Float(
     A=5
     B=&FloatFromCurve(FloatCurve=DataInterface() {propertyValues {Curve {keys [<at 0 0.8 RCIM_Cubic>,<at 1 0 RCIM_Cubic>]}}})
    )
    Scale RGB=&Multiply_VectorByFloat(
     Vector=&MakeVectorFromLinearColorRGB(Color=<rgba 0.9485 0.1 1 1>)
     Float=&FloatFromCurve(
      FloatCurve=DataInterface() {propertyValues {Curve {keys [<at 0.1654 2>,<at 0.2596 0.1609>,<at 0.3956 0>,<at 0.5 1.5>,<at 0.6 1.5>,<at 0.7029 0.016>,<at 0.8029 0.016>,<at 0.9005 0.3575>,<at 1 0>]}}}
      Scale Curve=3
     )
    )
    ScaleA=true
    ScaleRGB=true
   )
   &ScaleColor(
    Scale Alpha=&Multiply_Float(
     A=&FloatFromCurve(FloatCurve=DataInterface() {propertyValues {Curve {keys [<at 0 1 RCIM_Cubic>,<at 1 0 RCIM_Cubic>]}}}, Scale Curve=3)
     B=&FN_FrontEndFloatModifier(Gameplay=1, Front End=0.8)
    )
    Scale RGB=&Multiply_VectorByFloat(
     Vector=&MakeVectorFromLinearColorRGB(Color=<rgba 0.9485 0.1 1 1>)
     Float=&FloatFromCurve(
      FloatCurve=DataInterface() {propertyValues {Curve {keys [<at 0 1>,<at 0.2 0>,<at 0.3 1>,<at 0.5073 0.0115>,<at 0.8029 0.016>,<at 0.9005 0.3575>,<at 1 0>]}}}
      Scale Curve=4
     )
    )
    ScaleA=true
    ScaleRGB=true
   ) {enabled=false, module=ScaleColor001}
   &DynamicMaterialParameters(
    Index 0 Param 1=&FloatFromCurve(FloatCurve=DataInterface() {propertyValues {Curve {keys [{tangentMode=RCTM_User, time=0, value=0.2}, {tangentMode=RCTM_User, time=1, value=1.5}]}}})
    Index 0 Param 2=0.7
    Index 0 Param 3=&FloatFromCurve(FloatCurve=DataInterface() {propertyValues {Curve {keys [
         <at 0 0 RCIM_Cubic>
         {arriveTangent=6.0024, interpMode=RCIM_Cubic, leaveTangent=6.0024, tangentMode=RCTM_User, time=0.4463, value=1.4389}
         {arriveTangent=-0.3684, interpMode=RCIM_Cubic, leaveTangent=-0.3684, tangentMode=RCTM_User, time=1, value=3}
        ]}}})
    Index 0 Param 4=1
    Index 1 Param 1=10
    Index 1 Param 2=&RandomRangeFloat(Minimum=0.5, Maximum=1)
    Index 1 Param 3=&RandomRangeFloat(Minimum=0.8, Maximum=1.2)
    Index 1 Param 4=1
    Module.Param0WriteEnabled=true
    Module.Param0WriteEnabled1=true
    Module.Param1WriteEnabled=true
    Module.Param1WriteEnabled1=true
    Module.Param2WriteEnabled=true
    Module.Param2WriteEnabled1=true
    Module.Param3WriteEnabled=true
    Module.Param3WriteEnabled1=true
    Write Parameter Index 1=true
   )
   &SolveForcesAndVelocity(Clamp Velocity=false, Limit Acceleration=false)
   /SetVariables(Particles.Lightning glow color:LinearColor=&MakeLinearColorFromVectorAndFloat(
     "Vector (RGB)"=&MakeVectorFromLinearColorRGB(Color=<rgba 1.5 0.0226 0.0655 1>)
     "Float (Alpha)"=&Multiply_Float(
      A=&FloatFromCurve(
       FloatCurve=DataInterface() {propertyValues {Curve {keys [<at 0 1 RCIM_Cubic>, {interpMode=RCIM_Cubic, leaveTangent=-3.6307, time=0.9, value=0}]}}}
       Scale Curve=3
      )
      B=1
     )
    )) {isSetParameters=true}
   /SetVariables(Particles.Glow Dy02:Vector4f=&MakeVector4(
     X=&FloatFromCurve(FloatCurve=DataInterface() {propertyValues {Curve {keys [{tangentMode=RCTM_User, time=0, value=0.2}, {tangentMode=RCTM_User, time=0.5, value=1}]}}})
     Y=0.7
     Z=0
     W=1
    )) {enabled=false, isSetParameters=true}
   /SetVariables(Particles.Glow Dy03:Vector4f=&MakeVector4(
     X=&FloatFromCurve(FloatCurve=DataInterface() {propertyValues {Curve {keys [<at 0.5 1>,<at 1 0.3>]}}})
     Y=0
     Z=0.5
     W=1
    )) {enabled=false, isSetParameters=true}
   &FN_Sprite_Attributes(
    Camera Offset=20
    Color=&MakeLinearColorFromVectorAndFloat(
     "Vector (RGB)"=&Multiply_Vector(A=&MakeVectorFromLinearColorRGB(Color=<rgba 0.3787 0.02 1 1>), B=<1 1 1>)
     "Float (Alpha)"=&Multiply_Float(
      A=&FloatFromCurve(FloatCurve=DataInterface() {propertyValues {Curve {keys [
           {interpMode=RCIM_Cubic, tangentMode=RCTM_User, time=0, value=0}
           {interpMode=RCIM_Cubic, tangentMode=RCTM_User, time=0.2, value=1}
           {tangentMode=RCTM_User, time=0.6, value=0}
          ]}}})
      B=1
     )
    )
    Create Camera Offset Attribute=true
    Create Color Attribute=true
    Create SpriteSize Attribute=true
    SpriteSize=&Multiply_Vector2D_ByFloat(
     Float=&FloatFromCurve(FloatCurve=DataInterface() {propertyValues {Curve {keys [
          {interpMode=RCIM_Cubic, tangentMode=RCTM_User, time=0, value=0}
          {interpMode=RCIM_Cubic, tangentMode=RCTM_User, time=0.2, value=1}
          {tangentMode=RCTM_User, time=0.6, value=0}
         ]}}})
     Vector2D=<30 30>
    )
    Write to Module Namespace=true
   ) {enabled=false}
  ]
  rendererClasses [&NiagaraMeshRendererProperties]
  renderers [
   {rendererClass=&NiagaraMeshRendererProperties, rendererIndex=0}
   {rendererClass=&NiagaraMeshRendererProperties, rendererIndex=1}
  ]
 }
 PP_Shockwave {
  emitterUpdateScript [
   &EmitterState(Life Cycle Mode=::NewEnumerator1, Loop Behavior=::NewEnumerator1, Loop Duration=3, UseLoopDelay=false)
   &SpawnBurst_Instantaneous(Spawn Time=0.05, Use Spawn Probability=false, UseLoopCountLimit=false)
   &SpawnBurst_Instantaneous(Spawn Group=1, Use Spawn Probability=false, UseLoopCountLimit=false) {module=SpawnBurst_Instantaneous001}
  ]
  particleSpawnScript [
   &InitializeParticle(
    Lifetime=2
    Mesh Scale Mode=::NewEnumerator3
    Mesh Uniform Scale=10
    Sprite Size=&Multiply_Vector2D_ByFloat(Float=&Multiply_Float(A=$User.ElementSizeScale, B=$User.PostSizeScale), Vector2D=<5048.9736 3201.8015>)
    Sprite Size Mode=::NewEnumerator4
    UsePositionOffset=false
    Write Material Random=true
   )
   &CameraOffset()
   /SetVariables(Particles.VisibilityTag=&SpawnGroup()) {isSetParameters=true}
  ]
  particleUpdateScript [
   &ParticleState()
   &Timeline(
    Play Rate=&FloatFromCurve(
     FloatCurve=DataInterface() {propertyValues {Curve {keys [
         {interpMode=RCIM_Cubic, leaveTangent=-12.5334, tangentMode=RCTM_Break, time=0, value=4}
         {interpMode=RCIM_Cubic, tangentMode=RCTM_Break, time=0.7, value=1}
        ]}}}
     Scale Curve=0.692
    )
    Timeline End=1.0000000200408773e+20
   )
   &ScaleSpriteSize(
    Uniform Curve Scale=3
    Uniform Curve Sprite Scale=DataInterface() {propertyValues {Curve {keys [
        {arriveTangent=15.597, interpMode=RCIM_Cubic, leaveTangent=15.597, tangentMode=RCTM_User, time=0, value=0}
        {arriveTangent=0.3869, interpMode=RCIM_Cubic, leaveTangent=0.3869, tangentMode=RCTM_User, time=0.05, value=1}
        {arriveTangent=0.0224, interpMode=RCIM_Cubic, leaveTangent=0.0224, tangentMode=RCTM_User, time=1, value=1}
       ]}}}
   )
   &DynamicMaterialParameters(
    Index 0 Param 1=&FloatFromCurve(
     FloatCurve=DataInterface() {propertyValues {Curve {keys [
         {interpMode=RCIM_Cubic, leaveTangent=-4.8343, tangentMode=RCTM_Break, time=0, value=2}
         {arriveTangent=-0.1577, interpMode=RCIM_Cubic, tangentMode=RCTM_Break, time=1, value=0}
        ]}}}
     Scale Curve=&Multiply_Float(A=2, B=$User.PostNoiseStrength)
    )
    Index 0 Param 2=&FloatFromCurve(
     FloatCurve=DataInterface() {propertyValues {Curve {keys [
         {interpMode=RCIM_Cubic, leaveTangent=-16.2213, tangentMode=RCTM_Break, time=0, value=4}
         {arriveTangent=-17.124, interpMode=RCIM_Cubic, tangentMode=RCTM_Break, time=0.25, value=0}
        ]}}}
     Scale Curve=&Multiply_Float(A=1, B=$User.PostStrength)
    )
    Index 0 Param 3=1
    Index 0 Param 4=1
    Index 1 Param 1=0.1
    Index 1 Param 2=1500
    Index 1 Param 3=0.5
    Index 1 Param 4=1
    Index 2 Param 1=$Particles.Timeline.Time
    Index 2 Param 2=$Particles.Timeline.Time
    Index 2 Param 3=$Particles.Timeline.Time
    Index 2 Param 4=&Multiply_Float(A=0.2, B=$User.ElementSizeScale)
    Index 3 Param 1=&FloatFromCurve(FloatCurve=DataInterface() {propertyValues {Curve {keys [<at 0.1 1>,<at 0.4 0>]}}})
    Index 3 Param 2=200
    Module.Index0Float0Write=true
    Module.Index0Float1Write=true
    Module.Index0Float2Write=true
    Module.Index0Float3Write=true
    Module.Index1Float0Write=true
    Module.Index1Float1Write=true
    Module.Index1Float2Write=true
    Module.Index1Float3Write=true
    Module.Index2Float0Write=true
    Module.Index2Float1Write=true
    Module.Index2Float2Write=true
    Module.Index2Float3Write=true
    Module.Index3Float0Write=true
    Module.Index3Float1Write=true
    Module.Index3Float2Write=true
    Module.Index3Float3Write=true
    Write Parameter Index 1=::NewEnumerator1
    Write Parameter Index 2=::NewEnumerator1
    Write Parameter Index 3=::NewEnumerator1
   )
   &ScaleColor(
    Scale Alpha=&FloatFromCurve(
     FloatCurve=DataInterface() {propertyValues {Curve {keys [{interpMode=RCIM_Cubic, leaveTangent=-3.4164, tangentMode=RCTM_Break, time=0.2, value=1}, <at 1 0>]}}}
     Scale Curve=0.3
    )
    ScaleA=true
    ScaleRGB=true
   )
   /SetVariables(Particles.BlendTest:NiagaraFloat=&FloatFromCurve(
     FloatCurve=DataInterface() {propertyValues {Curve {keys [
         {arriveTangent=6.3982, interpMode=RCIM_Cubic, leaveTangent=6.3982, tangentMode=RCTM_User, time=0, value=0}
         {arriveTangent=0.0074, interpMode=RCIM_Cubic, leaveTangent=0.0074, tangentMode=RCTM_User, time=0.2, value=0.9847}
         {arriveTangent=-1.4231, interpMode=RCIM_Cubic, leaveTangent=-1.4231, tangentMode=RCTM_User, time=2, value=0}
        ]}}}
     CurveIndex=$Emitter.Age
     Scale Curve=&Subtract_Float(
      A=1
      B=&FN_LOD_Distance_Remap(Start Distance=4000, End Distance=6000, Minimum Output=0, Maximum Output=1, Clamp Output=true)
     )
    )) {isSetParameters=true}
   /SetVariables(Particles.ComponentsEnabled:NiagaraBool=&Boolean_NotOperation(Bool=&SetBoolByFloatComparison(
      A=&FN_LOD_Distance_Remap(Start Distance=4000, End Distance=6000, Minimum Output=0, Maximum Output=1, Clamp Output=true)
      B=0.99
     ))) {isSetParameters=true}
   &KillParticles(Kill Particles=&SetBoolByFloatComparison(
     A=&FloatFromCurve(FloatCurve=DataInterface() {propertyValues {Curve {keys [<at 0.4 1>,<at 0.41 -0.01>]}}})
     B=0
     Comparison Type=::NewEnumerator6
    ))
  ]
  rendererClasses [&NiagaraSpriteRendererProperties]
  renderers [{rendererClass=&NiagaraSpriteRendererProperties}]
 }
 Radial_Smoke {
  emitterUpdateScript [
   &EmitterState(Life Cycle Mode=::NewEnumerator1, Loop Behavior=::NewEnumerator1, Loop Delay=0.15, Loop Duration=1, UseLoopDelay=true)
   &SpawnBurst_Instantaneous(Use Spawn Probability=false, UseLoopCountLimit=false)
  ]
  particleSpawnScript [
   &InitializeParticle(
    Color=<rgba 0.035 0.0175 0.019 1>
    Mesh Scale Mode=::NewEnumerator3
    Mesh Uniform Scale=&Multiply_Float(A=20, B=$User.ElementSizeScale)
    Sprite Size Mode=::NewEnumerator3
    Uniform Sprite Size=&Multiply_Float(A=4130.5918, B=$User.ElementSizeScale)
    UsePositionOffset=false
    Write Material Random=true
   )
   &InitialMeshOrientation(Mesh Axis To Orient=<0 0 1>, Mesh Orientation Mode=::NewEnumerator2, Use Rotation Vector=false)
  ]
  particleUpdateScript [
   &ParticleState()
   &ScaleSpriteSize(Uniform Curve Sprite Scale=DataInterface() {propertyValues {Curve {keys [{tangentMode=RCTM_User, time=0, value=0.5}, {tangentMode=RCTM_User, time=0.2, value=1}]}}})
   &Timeline(
    Play Rate=&FloatFromCurve(FloatCurve=DataInterface() {propertyValues {Curve {keys [<at 0 2>,<at 1 1.3>]}}})
    Timeline End=1000
   )
   &CameraOffset(Camera Offset Amount=&Multiply_Float(
     A=-500
     B=&FloatFromCurve(
      FloatCurve=DataInterface() {propertyValues {Curve {keys [<at 0 1>,<at 1 0>]}}}
      CurveIndex=&FN_CameraDistance(
       0Dist:NiagaraFloat=500
       1Dist:NiagaraFloat=2000
       Camera Query:NiagaraDataInterfaceCamera=DataInterface()
       Position:NiagaraPosition=$Particles.Position
      )
     )
    ))
   &ScaleColor(
    Scale Alpha=&FloatFromCurve(FloatCurve=DataInterface() {propertyValues {Curve {keys [
         {arriveTangent=9.0725, interpMode=RCIM_Cubic, leaveTangent=9.0725, tangentMode=RCTM_User, time=0, value=0}
         {arriveTangent=0.1216, interpMode=RCIM_Cubic, leaveTangent=0.1216, tangentMode=RCTM_User, time=0.07, value=0.9847}
         {arriveTangent=-3.3987, interpMode=RCIM_Cubic, leaveTangent=-3.3987, tangentMode=RCTM_User, time=1, value=0}
        ]}}})
    ScaleA=true
    ScaleRGB=true
   ) {module=ScaleColor002}
   &DynamicMaterialParameters(
    Index 0 Param 1=$Particles.Timeline.Time
    Module.Index0Param0Write=true
    Module.Index0Param1Write=true
    Module.Index0Param2Write=true
    Module.Index0Param3Write=true
   )
  ]
  rendererClasses [&NiagaraSpriteRendererProperties]
  renderers [{rendererClass=&NiagaraSpriteRendererProperties}]
 }
 Ring_Shockwave {
  emitterUpdateScript [
   &EmitterState()
   &SpawnBurst_Instantaneous(Use Spawn Probability=false, UseLoopCountLimit=false)
  ]
  particleSpawnScript [
   &InitializeParticle(Lifetime=0.3, Mesh Scale Mode=::NewEnumerator3, Mesh Uniform Scale=5, UsePositionOffset=false, Write Material Random=true)
   &InheritVelocity(Inherited Velocity Speed Limit=1000, Speed Limit=true, UseSpeedThreshold=false)
  ]
  particleUpdateScript [
   &ParticleState()
   &ScaleMeshSize(Scale Factor=&VectorFromFloat(Value=&FloatFromCurve(
      FloatCurve=DataInterface() {propertyValues {Curve {
         defaultValue=0
         keys [
          {arriveTangent=2.2778, interpMode=RCIM_Cubic, leaveTangent=2.2778, tangentMode=RCTM_User, time=0.2042, value=0}
          {arriveTangent=0.0224, interpMode=RCIM_Cubic, leaveTangent=0.0224, tangentMode=RCTM_User, time=1, value=1}
         ]
        }}}
      Scale Curve=5
     )))
   &DynamicMaterialParameters(
    Index 0 Param 4=&FloatFromCurve(FloatCurve=DataInterface() {propertyValues {Curve {keys [<at 0 1 RCIM_Cubic>,<at 1 0 RCIM_Cubic>]}}}, Scale Curve=0.05)
    Module.Index0Float0Write=true
    Module.Index0Float1Write=true
    Module.Index0Float2Write=true
    Module.Index0Float3Write=true
   )
   &ScaleColor(
    Scale Alpha=&FloatFromCurve(FloatCurve=DataInterface() {propertyValues {Curve {keys [<at 0 1>,<at 1 0>]}}}, Scale Curve=0.5)
    ScaleA=true
    ScaleRGB=true
   )
   &Drag(Use Linear Drag=true, Use Rotational Drag=true)
   &SolveForcesAndVelocity(Clamp Velocity=false, Limit Acceleration=false)
  ]
  rendererClasses [&NiagaraMeshRendererProperties]
  renderers [{rendererClass=&NiagaraMeshRendererProperties, rendererIndex=0}]
 }
}
summary {
 systemName=NS_LargeSpire_DeathFX
 emitters [
  {emitterName=CenterDenseSmoke, rendererClasses [&NiagaraSpriteRendererProperties]}
  {emitterName=PP_Shockwave, rendererClasses [&NiagaraSpriteRendererProperties]}
  {
   emitterName=DissoveParticle001
   simTarget=GPUComputeSim
   rendererClasses [&NiagaraMeshRendererProperties, &NiagaraSpriteRendererProperties]
  }
  {emitterName=Radial_Smoke, rendererClasses [&NiagaraSpriteRendererProperties]}
  {
   emitterName=EnergyWave001
   rendererClasses [
    &NiagaraMeshRendererProperties
    &NiagaraLightRendererProperties
    &NiagaraSpriteRendererProperties
   ]
  }
  {emitterName=FN_Debris_Mesh, rendererClasses [&NiagaraMeshRendererProperties]}
  {emitterName=LightningSmall_BodyRound, rendererClasses [&NiagaraMeshRendererProperties]}
  {emitterName=Ring_Shockwave, rendererClasses [&NiagaraMeshRendererProperties]}
 ]
 userVariables [
  User.bIsFrontEnd {type=NiagaraBool}
  User.bIsFrontEndPreview {type=NiagaraBool}
  User.BrightParticleCountScale {type=NiagaraFloat, defaultValue=1, description=""}
  User.ElementSizeScale {type=NiagaraFloat, defaultValue=1, description=""}
  User.PostNoiseStrength {type=NiagaraFloat, defaultValue=1, description=""}
  User.PostSizeScale {type=NiagaraFloat, defaultValue=1, description=""}
  User.PostStrength {type=NiagaraFloat, defaultValue=1, description=""}
  User.RandomArray {
   type=NiagaraDataInterfaceArrayFloat
   defaultValue {
    struct=NiagaraExt_VariableValue_DataInterface
    value {
     dataInterface=&self:NiagaraDataInterfaceArrayFloat_0
     dataInterfaceClass=&NiagaraDataInterfaceArrayFloat
    }
   }
  }
  User.SmokeColorMul {type=LinearColor, description="", defaultValue {struct=LinearColor, value=<rgba 1 1 1 1>}}
  User.SmokeSizeScale {type=NiagaraFloat, defaultValue=1, description=""}
  User.UseNearCameraFade {type=NiagaraBool}
  User.VelocityScale {type=NiagaraFloat, defaultValue=1, description=""}
 ]
}
